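# Niagara swing trail for the EmeraldGlassGreen pickaxe, built from a single ribbon emitter, Lightning_Swing002.
# After a brief delay it spawns short-lived cyan ribbon particles for a fraction of a second, positioned by sampling the User.SkeletalMesh.
# Over life the ribbon narrows and fades through alpha, scaled by time of day, and dynamic material parameters, with a slight camera offset.
# The green RGB flicker curves are authored but switched off by ScaleRGB.
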NiagaraSystem /BRCosmetics/Weapons/FORT_Melee/Pickaxe_EmeraldGlassGreen/FX/NS_Pickaxe_EmeraldGlassGreen_Swing
dependencies {
 usedDataInterfaces [&NiagaraDataInterfaceColorCurve, &NiagaraDataInterfaceCurve]
 usedDynamicInputs []
 usedModules [
  &SystemState
  &self:SystemSpawnScript.SystemScriptSource.SystemScriptGraph.NiagaraNodeAssignment_0.SetVariables_#1
 ]
}
emitters {Lightning_Swing002 {
  emitterUpdateScript [
   &EmitterState(Life Cycle Mode=::NewEnumerator1, Loop Behavior=::NewEnumerator1, Loop Delay=0.07, Loop Duration=0.2, UseLoopDelay=true)
   &SpawnRate(SpawnRate=60, Use Spawn Probability=false)
  ]
  particleSpawnScript [
   &InitializeRibbon(
    Color=<rgba 0 0.4342 0.6514 50>
    Lifetime=0.3
    Ribbon Facing Mode:ENiagara_UnsetDirectSet=::NewEnumerator1
    Ribbon Facing Vector:Vector3f=<1 0 0>
    Ribbon Width=70
    UsePositionOffset=false
    Write Material Random=true
   )
   &SkeletalMeshLocation(
    Bone / Socket Index=2
    MaskSpawnGroup=false
    Mesh Sampling Type=::NewEnumerator5
    Sampled Position Offset=<10 0 0>
    Skeletal Mesh=$User.SkeletalMesh
    Socket Sampling Mode=::NewEnumerator5
   )
  ]
  particleUpdateScript [
   &ParticleState()
   &ScaleColor(
    Scale Alpha=&Multiply_Float(
     A=&FloatFromCurve(
      FloatCurve=DataInterface() {propertyValues {Curve {keys [{interpMode=RCIM_Cubic, tangentMode=RCTM_User, time=0, value=1}, {interpMode=RCIM_Cubic, tangentMode=RCTM_User, time=1, value=0}]}}}
      Scale Curve=&FNTimeOfDayScalar(Night Scale=0.2, Morning Scale=0.2, Evening Scale=0.2)
     )
     B=1
    )
    Scale RGB=&Multiply_VectorByFloat(
     Vector=&MakeVectorFromLinearColorRGB(Color=&ColorFromCurve(ColorCurve=DataInterface() {propertyValues {
         AlphaCurve {keys []}
         BlueCurve {keys [<at 0 4.8655 RCIM_Cubic>,<at 0.4581 4.8655>,<at 0.8818 6.0417>,<at 0.9992 2.9167 RCIM_Cubic>]}
         GreenCurve {keys [<at 0 10 RCIM_Cubic>,<at 0.4581 10>,<at 0.8818 10>,<at 0.9992 5 RCIM_Cubic>]}
         RedCurve {keys [<at 0 1.1979 RCIM_Cubic>,<at 0.4581 1.1979>,<at 0.8818 0.5>,<at 0.9992 0 RCIM_Cubic>]}
        }}))
     Float=&FloatFromCurve(
      FloatCurve=DataInterface() {propertyValues {Curve {keys [<at 0 1>,<at 0.1 1.2>,<at 0.3 1>,<at 0.5 0.5>,<at 0.7 1.5>,<at 1 1>]}}}
      Scale Curve=0.8
     )
    )
    ScaleA=true
    ScaleRGB=false
   )
   &ScaleRibbonWidth(Ribbon Width Scale=&FloatFromCurve(FloatCurve=DataInterface() {propertyValues {Curve {keys [<at 0.45 1>,<at 0.8 0.675>,<at 1 0.25>]}}})) {module=ScaleRibbonWidth001}
   &DynamicMaterialParameters(
    Index 0 Param 1=&FloatFromCurve(FloatCurve=DataInterface() {propertyValues {Curve {keys [
         {arriveTangent=-0.8731, interpMode=RCIM_Cubic, leaveTangent=-0.8731, tangentMode=RCTM_User, time=0.45, value=1}
         {arriveTangent=-1.6082, interpMode=RCIM_Cubic, leaveTangent=-1.6082, tangentMode=RCTM_User, time=1, value=0}
        ]}}})
    Index 0 Param 2=&FloatFromCurve(FloatCurve=DataInterface() {propertyValues {Curve {keys [
         {arriveTangent=-0.8731, interpMode=RCIM_Cubic, leaveTangent=-0.8731, tangentMode=RCTM_User, time=0, value=0}
         {arriveTangent=-0.012, interpMode=RCIM_Cubic, leaveTangent=-0.012, tangentMode=RCTM_User, time=1, value=-1}
        ]}}})
    Module.Param0WriteEnabled=true
    Module.Param1WriteEnabled=false
    Module.Param2WriteEnabled=false
    Module.Param3WriteEnabled=false
   ) {module=DynamicMaterialParameters001}
   &CameraOffset(Camera Offset Amount=10)
  ]
  rendererClasses [&NiagaraRibbonRendererProperties]
  renderers [{rendererClass=&NiagaraRibbonRendererProperties}]
 }}
summary {
 systemName=NS_Pickaxe_EmeraldGlassGreen_Swing
 emitters [{emitterName=Lightning_Swing002, rendererClasses [&NiagaraRibbonRendererProperties]}]
 userVariables [
  User.RibbonIDOffset {type=NiagaraInt32}
  User.SkeletalMesh {
   type=NiagaraDataInterfaceSkeletalMesh
   defaultValue {
    struct=NiagaraExt_VariableValue_DataInterface
    value {dataInterface=&self:NiagaraDataInterfaceSkeletalMesh_1}
   }
  }
  User.Type_Transformation {type=NiagaraFloat, defaultValue=0, description=""}
 ]
}
