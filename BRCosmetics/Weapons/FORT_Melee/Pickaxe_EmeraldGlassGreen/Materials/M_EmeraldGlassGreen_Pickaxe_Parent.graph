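# Unlit parent material for the Emerald Glass Green pickaxe, giving it a dark green toon look with a pencil texture.
# Character PreFX output feeds MF_AnimeShading, whose shading value blends two color parameters.
# That color is multiplied by the pencil texture's red channel, which UsePencilTexture fades toward white as it rises, despite its name.
# The result becomes Emissive Color and passes through the character PostFX function.
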Material /BRCosmetics/Weapons/FORT_Melee/Pickaxe_EmeraldGlassGreen/Materials/M_EmeraldGlassGreen_Pickaxe_Parent
graph {
 properties {
  bAutomaticallySetUsageInEditor=false
  bUseMaterialAttributes=true
  bUsedWithClothing=true
  bUsedWithMeshDeformer=true
  bUsedWithSkeletalMesh=true
  pixelDepthOffsetMode=PDOM_Legacy
  shadingModel=MSM_Unlit
  usedShadingModels=Unlit
 }
 nodes [
  n1 = MaterialFunctionCall() {materialFunction=&BaseFN_CharacterFunction_PreFX}
  n2 = LinearInterpolate(
   A=VectorParameter() {defaultValue=<rgba 0.024 0.0343 0.0309 1>}.RGB
   B=VectorParameter() {parameterName=Param_1, defaultValue=<rgba 0.009 0.0191 0.02 1>}.RGB
   Alpha=MaterialFunctionCall!(
    Material=@n1.Result
    LitDiffuseOverride
    SSC_AlphaOverride
    Tint Mask
    DF InkLine Visibility Mask
    DF Inkline Injection
    DF Inkline Color Override
    DF Inkline Color Override Mask
    CustomUVs
   ) {materialFunction=/Game/Packages/Fortress_SharedMaterials/Anime_Master_Material/MF_AnimeShading}.Shading
  )
  n3 = LinearInterpolate(
   A=TextureSample() {texture=/BRCosmetics/Weapons/FORT_Melee/Pickaxe_EmeraldGlassGreen/Textures/T_PencilTexture}.R
   Alpha=ScalarParameter() {parameterName=UsePencilTexture}
  )
  MaterialFunctionCall(
   AO=@n1.AO
   Distance Blend=@n1.AO
   Fresnel=@n1.AO
   Gmap=@n1.AO
   Pre Skinned Local Position=@n1.AO
   Material=SetMaterialAttributes(Emissive Color=Multiply(A=@n2, B=@n3)) {attributeSetTypes [#1]}
  ) {materialFunction=&BaseFN_CharacterFunction_PostFX}
 ]
}
